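# Sound cue for the BananaPhilosopher pickaxe cosmetic's enemy impact, using pickaxe impact attenuation and the impacts sound class.
# After a short fixed delay, it mixes a random base impact with a random gliss or high-pluck layer, with slight volume modulation.
# A quieter, further-delayed break layer joins only occasionally, because weighted empty slots in its Random node usually select silence.
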SoundCue /BRCosmetics/Sounds/PickAxes/BananaPhilosopher/PickaxeImpactEnemy_BananaPhilosopher_Cue
root=Delay_0
nodes [
 {id=Delay_0, children [Mixer_0], properties {delayMax=0.1, delayMin=0.1}}
 {id=Delay_1, children [Random_3], properties {delayMax=0.08, delayMin=0.05}}
 {id=Mixer_0, children [Modulator_0,Modulator_1], properties {inputVolume [1,1.2]}}
 {id=Mixer_2, children [Delay_1,Random_2], properties {inputVolume [0.5,1]}}
 {id=Modulator_0, children [Random_0], properties {pitchMax=1, pitchMin=1, volumeMax=1.02, volumeMin=0.98}}
 {id=Modulator_1, children [Mixer_2], properties {pitchMax=1, pitchMin=1, volumeMax=1.02, volumeMin=0.98}}
 {
  id=Random_0
  children [WavePlayer_16, WavePlayer_17, WavePlayer_18, WavePlayer_19, WavePlayer_20, WavePlayer_21]
  properties {weights [1,1,1,1,1,1]}
 }
 {
  id=Random_2
  children [WavePlayer_0,WavePlayer_1,WavePlayer_2,WavePlayer_3,WavePlayer_4,WavePlayer_5,WavePlayer_9,WavePlayer_8,WavePlayer_7,WavePlayer_6]
  properties {weights [1,1,1,1,1,1,1,1,1,1]}
 }
 {
  id=Random_3
  children [WavePlayer_28, WavePlayer_29, WavePlayer_30, WavePlayer_31, WavePlayer_32, WavePlayer_33, null, null, null, null]
  properties {weights [0.3,0.3,0.3,0.3,0.3,0.3,1,1,1,1]}
 }
 {id=WavePlayer_0, properties {soundWaveAssetPtr=&/Pickaxe_BananaPhilosopher_Gliss_01}}
 {id=WavePlayer_1, properties {soundWaveAssetPtr=&/Pickaxe_BananaPhilosopher_Gliss_02}}
 {id=WavePlayer_16, properties {soundWaveAssetPtr=&/Pickaxe_BananaPhilosopher_BaseImpact_01}}
 {id=WavePlayer_17, properties {soundWaveAssetPtr=&/Pickaxe_BananaPhilosopher_BaseImpact_02}}
 {id=WavePlayer_18, properties {soundWaveAssetPtr=&/Pickaxe_BananaPhilosopher_BaseImpact_03}}
 {id=WavePlayer_19, properties {soundWaveAssetPtr=&/Pickaxe_BananaPhilosopher_BaseImpact_04}}
 {id=WavePlayer_2, properties {soundWaveAssetPtr=&/Pickaxe_BananaPhilosopher_Gliss_03}}
 {id=WavePlayer_20, properties {soundWaveAssetPtr=&/Pickaxe_BananaPhilosopher_BaseImpact_05}}
 {id=WavePlayer_21, properties {soundWaveAssetPtr=&/Pickaxe_BananaPhilosopher_BaseImpact_06}}
 {id=WavePlayer_28, properties {soundWaveAssetPtr=&/Pickaxe_BananaPhilosopher_Break_01}}
 {id=WavePlayer_29, properties {soundWaveAssetPtr=&/Pickaxe_BananaPhilosopher_Break_02}}
 {id=WavePlayer_3, properties {soundWaveAssetPtr=&/Pickaxe_BananaPhilosopher_Gliss_04}}
 {id=WavePlayer_30, properties {soundWaveAssetPtr=&/Pickaxe_BananaPhilosopher_Break_03}}
 {id=WavePlayer_31, properties {soundWaveAssetPtr=&/Pickaxe_BananaPhilosopher_Break_04}}
 {id=WavePlayer_32, properties {soundWaveAssetPtr=&/Pickaxe_BananaPhilosopher_Break_05}}
 {id=WavePlayer_33, properties {soundWaveAssetPtr=&/Pickaxe_BananaPhilosopher_Break_06}}
 {id=WavePlayer_4, properties {soundWaveAssetPtr=&/Pickaxe_BananaPhilosopher_Gliss_05}}
 {id=WavePlayer_5, properties {soundWaveAssetPtr=&/Pickaxe_BananaPhilosopher_Gliss_06}}
 {id=WavePlayer_6, properties {soundWaveAssetPtr=&/Pickaxe_BananaPhilosopher_HighPluck_Impact_04}}
 {id=WavePlayer_7, properties {soundWaveAssetPtr=&/Pickaxe_BananaPhilosopher_HighPluck_Impact_03}}
 {id=WavePlayer_8, properties {soundWaveAssetPtr=&/Pickaxe_BananaPhilosopher_HighPluck_Impact_02}}
 {id=WavePlayer_9, properties {soundWaveAssetPtr=&/Pickaxe_BananaPhilosopher_HighPluck_Impact_01}}
]
properties {
 attenuationSettings=/Game/Sounds/Attenuation/Fort_Pickaxe_Impact_Att
 audioPropertiesSheet=/BRCosmetics/Sounds/Settings/PropertySheets/Pickaxe/APS_Cosmetic_Pickaxe_Impact
 maxDistance=4300
 soundClassObject=&Fort_Impacts
 volumeMultiplier=0.45
 concurrencySet [null]
}
